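# Front-end UI sound cue that, judging by its name and wave, plays a whoosh when rewards header text appears.
# A single wave player feeds the root directly, with no randomization or modulation, under the Fort_UI sound class.
# A concurrency override allows only one instance at a time and rejects new plays while it sounds, so repeated triggers do not stack.
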SoundCue /Game/Sounds/Fort_UI_Sounds/FrontEnd/New_FrontEnd/UI_ShowHeaderText_Appear_Cue
root=WavePlayer_0
nodes [{id=WavePlayer_0, properties {soundWaveAssetPtr=&/UI_Rewards_ShowHeaderText_Appear_Whoosh_01}}]
properties {
 bOverrideConcurrency=true
 maxDistance=2097152
 soundClassObject=&Fort_UI
 concurrencyOverrides {maxCount=1, resolutionRule=PreventNew}
}
