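# Translucent UI material that draws a two-color linear gradient with optional shine and edge fade.
# Rotated, offset UVs drive the gradient, which can blend toward a curve atlas remap before a power adjustment mixes both color and alpha.
# A time-looped, diagonally scrolling shine texture can tint the color toward ShineColor.
# GradientCornerFade can limit alpha using a smoothstepped U gradient, soft-light blended with vector noise.
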Material /Game/UI/Style/Materials/BasicLinearUIGradient
graph {
 properties {
  bUsedWithStaticMesh=false
  blendMode=BLEND_Translucent
  materialDomain=MD_UI
  pixelDepthOffsetMode=PDOM_Legacy
  usedShadingModels=Default Lit
 }
 nodes [
  n1 = VectorParameter() {parameterName=Color1, defaultValue=<rgba 1 0.4 0 1>}
  n2 = VectorParameter() {parameterName=Color2, defaultValue=<rgba 0 0.99 0.2173 0>}
  n3 = TextureCoordinate()
  n4 = VectorParameter() {parameterName=UVOffset, defaultValue=<rgba 0 0 0 1>}
  n5 = AppendVector(A=@n4.R, B=@n4.R)
  n6 = MaterialFunctionCall(UV Channel=MaterialFunctionCall(UVs=Add(A=@n3, B=@n5), "Rotation Angle (0-1)"=ScalarParameter() {parameterName=Rotation, sliderMax=1}) {materialFunction=&CustomRotator}.Rotated Values) {materialFunction=&LinearGradient}
  n7 = Power(
   Base=LinearInterpolate(
    A=@n6.VGradient
    B=CurveAtlasRowParameter(CurveTime=@n6.VGradient) {atlas=/Game/UI/SocialPanel/Art/CA_SocialPanel, curve=/Game/UI/Foundation/RockyRidge/C_RockyRidgeV2, defaultValue=11}.R
    Alpha=ScalarParameter() {parameterName=UsCurveAsset}
   )
   Exp=ScalarParameter() {defaultValue=1, parameterName=GradientPow}
  )
  n8 = LinearInterpolate(A=@n1.RGB, B=@n2.RGB, Alpha=@n7)
  n9 = VectorParameter() {parameterName=ShineColor, defaultValue=<rgba 1 1 1 1>}
  n10 = TextureCoordinate()
  n11 = Constant() {r=-0.05}
  n12 = Time() {period=1}
  n13 = LinearInterpolate(Alpha=Fmod(A=Multiply(A=@n12) {constB=0.2}, B=Constant() {r=1})) {constA=1, constB=-10}
  n14 = TextureSample(UVs=Add(
    A=MaterialFunctionCall(UVs=@n10, "Rotation Angle (0-1)"=@n11) {materialFunction=&CustomRotator}.Rotated Values
    B=AppendVector(A=@n13, B=@n13)
   )) {samplerSource=SSM_Clamp_WorldGroupSettings, texture=/Game/UI/Foundation/Textures/Pattern/T-UI-Shine}
  LinearInterpolate(A=@n8, B=LinearInterpolate(A=@n8, B=@n9.RGB, Alpha=@n14.RGB), Alpha=ScalarParameter() {parameterName=UseShine})
  n15 = LinearInterpolate(A=@n1.A, B=@n2.A, Alpha=@n7) {constA=0.96, constB=0.7}
  n16 = TextureCoordinate()
  n17 = SmoothStep(
   Min=ScalarParameter() {defaultValue=0.2, parameterName=GradientCornerFade_Min}
   Max=ScalarParameter() {defaultValue=0.7, parameterName=GradientCornerFade_Max}
   Value=MaterialFunctionCall(UV Channel=OneMinus(None=@n16)) {materialFunction=&LinearGradient}.UGradient
  ) {constMax=0.7, constMin=0.2}
  n18 = Multiply(A=VectorNoise(), B=ScalarParameter() {defaultValue=0.25, parameterName=NoiseAmount}) {constB=0.1}
  LinearInterpolate(
   A=@n15
   B=Min(
    A=@n15
    B=MaterialFunctionCall!(Base=@n17, Blend=@n18) {materialFunction=/Engine/Functions/Engine_MaterialFunctions03/Blends/Blend_SoftLight}.Result
   )
   Alpha=ScalarParameter() {parameterName=GradientCornerFade}
  ) {constAlpha=1}
  Multiply(A=@n14.RGB, B=@n9.RGB)
  n19 = ScalarParameter() {defaultValue=7, parameterName=GradientCornerFade_Multiplier}
  Multiply(A=Constant() {r=0.01}, B=@n19)
  Multiply(A=Constant() {r=0.1}, B=@n19)
 ]
}
