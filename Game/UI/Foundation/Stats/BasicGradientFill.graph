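# BasicGradientFill is a Fortnite UI user widget: a single Image that draws a configurable two-color linear gradient.
# PreConstruct applies the stored size, colors and rotation through Set Brush Size and Set Fill.
# Set Brush Size rebuilds the image brush with the new size, copying most other fields.
# Set Fill clamps rotation to 0–1 and writes it and both colors to a dynamic instance of the BasicLinearUIGradient material.
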WidgetBlueprint /Game/UI/Foundation/Stats/BasicGradientFill
graphs [
 EventGraph {nodes [
   Event PreConstruct()
   > Set Brush Size!(execute, self, Brush Size:Vector 2D Structure=Get Brush Size() -> (Brush Size:Vector 2D Structure)) -> (then) {is=title K2Node_CallFunction}
   > Set Fill!(
    execute
    self
    Color 1:Linear Color Structure=Get Color 1() -> (Color 1:Linear Color Structure)
    Color 2:Linear Color Structure=Get Color 2() -> (Color 2:Linear Color Structure)
    "Rotation (0_-_1)":double=Get Rotation() -> (Rotation:double)
   ) -> (then) {is=title K2Node_CallFunction}
  ]}
 Set Brush Size {nodes [
   n1 = Set Brush Size!() -> (then, Brush Size:Vector 2D Structure) {is=title K2Node_FunctionEntry}
   > n2 = Set Brush Size(Brush Size:Vector 2D Structure=@n1) -> (Output_Get:Vector 2D Structure) {is=title K2Node_VariableSet}
   n3 = Get ImageFill() -> (ImageFill:Image Object Reference)
   n4 = Break Slate Brush!(SlateBrush:"Slate Brush Structure (by ref)"=Get Brush!(self:Image Object Reference=@n3) -> (Brush:Slate Brush Structure)) -> (
    TintColor:Slate Color Structure
    DrawAs:ESlateBrushDrawType Enum
    Tiling:ESlateBrushTileType Enum
    Mirroring:ESlateBrushMirrorType Enum
    Margin:Margin Structure
    ResourceObject:Object Reference
    OutlineSettings:Slate Brush Outline Settings Structure
   ) {is=title K2Node_BreakStruct}
   > SetBrush(
    self=@n3
    InBrush=Make Slate Brush!(
     TintColor:Slate Color Structure=@n4
     DrawAs:ESlateBrushDrawType Enum=@n4.DrawAs
     Tiling:ESlateBrushTileType Enum=@n4.Tiling
     Mirroring:ESlateBrushMirrorType Enum=@n4.Mirroring
     ImageSize:Vector 2D Structure=@n2
     Margin:Margin Structure=@n4.Margin
     ResourceObject:Object Reference=@n4.ResourceObject
     OutlineSettings:Slate Brush Outline Settings Structure=(CornerRadii=(X=0,Y=0,Z=0,W=0),Color=(SpecifiedColor=(R=0,G=0,B=0,A=0),ColorUseRule=UseColor_Specified),Width=0,RoundingType=HalfHeightRadius,bUseBrushTransparency=False)
    ) -> (SlateBrush:Slate Brush Structure) {is=title K2Node_MakeStruct}
   )
  ]}
 Set Fill {nodes [
   n1 = Set Fill!() -> (then, Color 1:Linear Color Structure, Color 2:Linear Color Structure, "Rotation (0_-_1)":double=~) {is=title K2Node_FunctionEntry}
   > Set Color 1(Color 1:Linear Color Structure=@n1) -> (Output_Get:Linear Color Structure)
   > Set Color 2(Color 2:Linear Color Structure=@n1.Color 2) -> (Output_Get:Linear Color Structure)
   > Set Rotation(Rotation:double="Clamp (Float)"(Value=@n1."Rotation (0_-_1)", Min=0, Max=1)) -> (Output_Get:double)
   > n2 = GetDynamicMaterial(self=Get ImageFill())
   > Is Valid(InputObject=@n2)
   > SetVectorParameterValue(self=@n2, ParameterName=Color1, Value=Get Color 1())
   > SetVectorParameterValue(self=@n2, ParameterName=Color2, Value=Get Color 2())
   > SetScalarParameterValue(self=@n2, ParameterName=Rotation, Value=Get Rotation())
  ]}
]
variables [
 Brush Size {type=Vector2D, default=<1 1>}
 Color 1 {type=LinearColor, default=<rgba 0 0.1779 0.2502 1>}
 Color 2 {type=LinearColor, default=<rgba 0.0452 0.2831 0.3325 1>}
 Rotation {type=number}
]
widgets {
 info {inheritedWidgetCount=0, parentClass=/Script/FortniteUI.FortUserWidget, rootWidgetClass=&Image, widgetCount=1}
 widgets [ImageFill {
   widgetClassPath=&Image
   bIsVariable=true
   widget=ImageFill
   properties {brush {
     bIsDynamicallyLoaded=false
     drawAs=Image
     imageType=NoImage
     mirroring=NoMirror
     resourceName=None
     resourceObject=/Game/UI/Style/Materials/BasicLinearUIGradient
     tiling=NoTile
     imageSize=<1 1>
     margin=<ltrb 0 0 0 0>
     outlineSettings {
      bUseBrushTransparency=false
      roundingType=HalfHeightRadius
      width=0
      color {colorUseRule=UseColor_Specified, specifiedColor=<rgba 0 0 0 0>}
      cornerRadii=<0 0 0 0>
     }
     tintColor {colorUseRule=UseColor_Specified, specifiedColor=<rgba 1 1 1 1>}
     uVRegion {bIsValid=false, max=<0 0>, min=<0 0>}
    }}
  }]
}
