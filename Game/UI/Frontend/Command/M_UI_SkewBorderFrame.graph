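# Translucent UI material that draws a skewed border frame, masking the area outside an irregular, slightly tilted quad.
# Four UV copies, offset and rotated by small angles around corner pivots, give the slanted edges, with the top/bottom pair reading UV channel 4 and the sides channel 0.
# SmoothStep sharpens each pair's product; their combined mask is inverted, clamped and scaled by AlphaOverride, with an RGB parameter for color.
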Material /Game/UI/Frontend/Command/M_UI_SkewBorderFrame
graph {
 properties {
  bUsedWithStaticMesh=false
  blendMode=BLEND_Translucent
  materialDomain=MD_UI
  pixelDepthOffsetMode=PDOM_Legacy
  usedShadingModels=Default Lit
 }
 nodes [
  n1 = TextureCoordinate() {coordinateIndex=4}
  n2 = Constant() {r=-0.06}
  n3 = Constant() {r=-1}
  n4 = ComponentMask(None=Rotator(Coordinate=Add(A=@n1, B=@n2), Time=Divide(A=@n3) {constA=360, constB=360}) {centerX=1, centerY=0, speed=6.2832}) {g=true}
  n5 = Add(A=@n1, B=Constant() {r=0.07})
  n6 = Divide(A=Constant() {r=-1.3}) {constA=360, constB=360}
  n7 = ComponentMask(None=OneMinus(None=Rotator(Coordinate=@n5, Time=@n6) {centerX=0, centerY=1, speed=6.2832})) {g=true}
  n8 = Constant() {r=0.0025}
  n9 = TextureCoordinate()
  n10 = Constant() {r=-0.05}
  n11 = Constant() {r=2}
  n12 = ComponentMask(None=Rotator(Coordinate=Add(A=@n9, B=@n10), Time=Divide(A=@n11) {constA=360, constB=360}) {centerX=0, centerY=1, speed=6.2832}) {r=true}
  n13 = Add(A=@n9, B=Constant() {r=0.04})
  n14 = Divide(A=Constant() {r=1.2}) {constA=360, constB=360}
  n15 = ComponentMask(None=OneMinus(None=Rotator(Coordinate=@n13, Time=@n14) {centerX=1, centerY=0, speed=6.2832})) {r=true}
  n16 = DotProduct(
   A=MaterialFunctionCall(Alpha=DotProduct(A=@n4, B=@n7), Max=@n8) {materialFunction=&SmoothStep}.Result
   B=MaterialFunctionCall(Alpha=DotProduct(A=@n12, B=@n15), Max=@n8) {materialFunction=&SmoothStep}.Result
  )
  Multiply(A=ScalarParameter() {defaultValue=0.175, parameterName=AlphaOverride}, B=Clamp(None=OneMinus(None=@n16)))
  VectorParameter() {parameterName=RGB, defaultValue=<rgba 1 1 1 1>}
 ]
}
