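# Player power rating widget that shows team power, an optional parenthesized base rating and an optional progress bar, in regular or oversized layout.
# Rating events set the base text, interpolate the team number and split the bar into base and team sections.
# UpdateWidgetState runs on construct and team changes, applying team-dependent glow, styles and dual-value visibility, plus colors.
# It also plays a gated sound when base power differs from the cached value.
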WidgetBlueprint /Game/UI/Foundation/Stats/Moonbeam_PowerRatingWidget
graphs [
 EventGraph {nodes [
   n1 = Event PreConstruct()
   > n2 = Branch(Condition=@n1.IsDesignTime)
   SetActiveWidget(
    execute=[@n3.then, @n2.then]
    self:Widget Switcher Object Reference=Get ModeTypeSwitcher() -> (ModeTypeSwitcher:Common Widget Switcher Legacy Object Reference)
    Widget=Select(
     Option 0:Widget Object Reference=Get RegularMode() -> (RegularMode:Horizontal Box Object Reference)
     Option 1:Widget Object Reference=Get OversizedMode() -> (OversizedMode:Vertical Box Object Reference)
     Index=Get UseOversizedMode() -> (UseOversizedMode:bool)
    ) -> (ReturnValue:Widget Object Reference)
   )
   > SetVisibility(
    self=[
     Get BaseRatingValue() -> (BaseRatingValue:Common Text Object Reference)
     Get BaseRatingValue-Large() -> (BaseRatingValue-Large:Common Text Object Reference)
    ]
    InVisibility=Select(
     Option 0:ESlateVisibility Enum=Collapsed
     Option 1:ESlateVisibility Enum=SelfHitTestInvisible
     Index=Get ShowDualPower() -> (ShowDualPower:bool)
    ) -> (ReturnValue:ESlateVisibility Enum)
   )
   > SetVisibility(
    self=Get RatingBar() -> (RatingBar:Size Box Object Reference)
    InVisibility=Select(Option 0=Collapsed, Option 1=SelfHitTestInvisible, Index=Get ShowProgress() -> (ShowProgress:bool))
   )
   > UpdateWidgetState()
   Event OnTeamMembershipChanged() {is=title K2Node_Event}
   > UpdateWidgetState()
   n3 = Event Construct()
   n4 = Event OnBasePowerRatingUpdated!() -> (OutputDelegate, then, BasePowerRating:int=0) {is=title K2Node_Event}
   > "SetText (Text)"(
    self=[Get BaseRatingValue(), Get BaseRatingValue-Large()]
    InText=Format Text!(Format="LOCTEXT(\"({Rating})\")", Rating:int=@n4.BasePowerRating) -> (Result)
   )
   n5 = Event OnTeamPowerRatingUpdated!() -> (OutputDelegate, then, TeamPowerRating:int=0) {is=title K2Node_Event}
   n6 = Get InterpDuration() -> (InterpDuration:double)
   > n7 = Branch(Condition="float > float"(A=@n6, B=0) {is=title K2Node_CallFunction})
   n8 = Get TeamRatingValue() -> (TeamRatingValue:Common Numeric Text Block Object Reference)
   n9 = Get TeamRatingValue-Large() -> (TeamRatingValue-Large:Common Numeric Text Block Object Reference)
   n10 = "To Float (Integer)"(InInt=@n5.TeamPowerRating)
   > InterpolateToValue!(
    execute
    self:Common Numeric Text Block Object Reference=[@n8, @n9]
    TargetValue:float=@n10
    MaximumInterpolationDuration:float=@n6
    MinimumChangeRate:float=1
    OutroOffset:float=0
   ) -> (then)
   SetCurrentValue!(execute=@n7.else, self:Common Numeric Text Block Object Reference=[@n8, @n9], NewValue:float=@n10) -> (then)
   n11 = Event OnPowerRatingProgressUpdated!() -> (OutputDelegate, then, BaseRatingProgress:float, TeamRatingProgress:float) {is=title K2Node_Event}
   > n12 = Branch(Condition=HasTeamMembers!(self:Fort Player Power Rating Object Reference) -> (ReturnValue:bool))
   > n13 = Branch(Condition="Equal (Integer)"(
     A=GetBasePowerRating!(self:Fort Player Power Rating Object Reference) -> (ReturnValue:int=0)
     B=GetTeamPowerRating!(self:Fort Player Power Rating Object Reference) -> (ReturnValue:int=0)
    ) {is=title K2Node_CallFunction})
   n14 = Get HomebaseRatingBar() -> (HomebaseRatingBar:Homebase Rating Bar Object Reference)
   > SetProgressBarSectionPercent!(
    execute
    self:Fort Material Progress Bar Object Reference=@n14
    BarSection:EFortMaterialProgressBarSection Enum=Primary
    Percent:float=@n11.BaseRatingProgress
   ) -> (then)
   > SetProgressBarSectionPercent(self=@n14, BarSection=Secondary, Percent=float - float(A=@n11.BaseRatingProgress, B=@n11.TeamRatingProgress) {is=title K2Node_CallFunction})
   n15 = Get HomebaseRatingBar()
   SetProgressBarSectionPercent(execute=@n13.else, self=@n15, BarSection=Primary, Percent=0)
   > SetProgressBarSectionPercent(self=@n15, BarSection=Secondary, Percent=@n11.TeamRatingProgress)
   n16 = Get HomebaseRatingBar()
   SetProgressBarSectionPercent(execute=@n12.else, self=@n16, BarSection=Primary, Percent=@n11.BaseRatingProgress)
   > SetProgressBarSectionPercent(self=@n16, BarSection=Secondary, Percent=0)
  ]}
 SetBackgroundColor {nodes [
   n1 = SetBackgroundColor!() -> (then, NewColor:Linear Color Structure) {is=title K2Node_FunctionEntry}
   > Set PowerBackingColor(PowerBackingColor:Linear Color Structure=@n1) -> (Output_Get:Linear Color Structure)
   > UpdateWidgetState()
  ]}
 UpdateWidgetState {nodes [
   UpdateWidgetState() {is=title K2Node_FunctionEntry}
   > Set IsContributingToTeamPowerRating(IsContributingToTeamPowerRating:bool=HasTeamMembers()) -> (Output_Get:bool)
   > n1 = Branch(Condition=AND Boolean(A=Get IsContributingToTeamPowerRating() -> (IsContributingToTeamPowerRating:bool), B=Get ShowGlow() -> (ShowGlow:bool)))
   n2 = Get BuffedRatingOuterGlow() -> (BuffedRatingOuterGlow:Image Object Reference)
   n3 = Get PowerIconGlow() -> (PowerIconGlow:Image Object Reference)
   n4 = Get Rating-Glow-Large() -> (Rating-Glow-Large:Image Object Reference)
   > SetVisibility(self=[@n2, @n3, @n4], InVisibility=Visible)
   n5 = Get BaseRatingValue-Large()
   n6 = Get TeamRatingValue-Large()
   > SetStyle!(
    execute
    self:Common Text Object Reference=[@n5, @n6]
    InStyle:Common Text Style Class Reference=/FortUILibrary/Styles/Typography/Foundation/TextStyle-Header-XL-BrightBlue.TextStyle-Header-XL-BrightBlue_C
   ) -> (then)
   n7 = Get UseLargeStyle() -> (UseLargeStyle:bool)
   > n8 = Branch(Condition=@n7)
   n9 = Get TeamRatingValue()
   > SetStyle(self=@n9, InStyle=/FortUILibrary/Styles/Typography/Foundation/TextStyle-Header-HomeBasePower.TextStyle-Header-HomeBasePower_C)
   n10 = Get BaseRatingValue()
   > SetStyle(self=@n10, InStyle=/FortUILibrary/Styles/Typography/Foundation/TextStyle-Header-L.TextStyle-Header-L_C)
   > n11 = SetBrush(
    self=Get Image_Power() -> (Image_Power:Image Object Reference)
    InBrush=Make Slate Brush!(
     TintColor:Slate Color Structure=(SpecifiedColor=(R=1,G=1,B=1,A=1),ColorUseRule=UseColor_Specified)
     DrawAs:ESlateBrushDrawType Enum=Image
     Tiling:ESlateBrushTileType Enum=NoTile
     Mirroring:ESlateBrushMirrorType Enum=NoMirror
     ImageSize:Vector 2D Structure=(X=56,Y=56)
     Margin:Margin Structure=(Left=0,Top=0,Right=0,Bottom=0)
     ResourceObject:Object Reference=/Game/UI/Foundation/Textures/Icons/Shell/T-Icon-Power-64
     OutlineSettings:Slate Brush Outline Settings Structure=(CornerRadii=(X=0,Y=0,Z=0,W=0),Color=(SpecifiedColor=(R=0,G=0,B=0,A=0),ColorUseRule=UseColor_Specified),Width=0,RoundingType=HalfHeightRadius,bUseBrushTransparency=False)
    ) -> (SlateBrush:Slate Brush Structure) {is=title K2Node_MakeStruct}
   )
   n12 = Get ShowDualPower()
   n13 = Get IsContributingToTeamPowerRating()
   SetVisibility(
    execute=[@n11.then, @n25.then, @n24.then, @n20.then]
    self=[Get BaseRatingValue(), Get BaseRatingValue-Large()]
    InVisibility=Select(Option 0=Collapsed, Option 1=SelfHitTestInvisible, Index=AND Boolean(A=@n13, B=@n12))
   )
   > SetColorAndOpacity(
    self:Image Object Reference=Get Moonbeam_Power_Back() -> (Moonbeam_Power_Back:Image Object Reference)
    InColorAndOpacity:Linear Color Structure=Get PowerBackingColor() -> (PowerBackingColor:Linear Color Structure)
   )
   > SetColorAndOpacity(
    self:Text Object Reference=[Get BaseRatingValue(), Get TeamRatingValue-Large(), Get TeamRatingValue()]
    InColorAndOpacity:Slate Color Structure=Get FontColor() -> (FontColor:Slate Color Structure)
   )
   n14 = HasTeamMembers()
   > n15 = Branch(Condition=@n14)
   n16 = "Not Equal (Integer)"!(A:int=GetBasePowerRating() -> (ReturnValue=0), B:int=Get Power Rating Num() -> (Power Rating Num:int=0)) -> (ReturnValue:bool) {is=title K2Node_CallFunction}
   > n17 = Branch(Condition=AND Boolean(A=@n14, B=@n16))
   PlaySound2D(execute=[@n17.then, @n18.Completed], Sound=Get PowerIncreasedSound() -> (PowerIncreasedSound:Sound Base Object Reference))
   > Set Power Rating Num(Power Rating Num:int=GetBasePowerRating() -> (ReturnValue=0)) -> (Output_Get:int=0)
   n18 = Do Once(execute=@n19.then, Reset=@n17.else, Start Closed=true)
   n19 = Branch(execute=@n15.else, Condition=@n16)
   SetStyle(
    execute=@n8.else
    self=@n9
    InStyle=/FortUILibrary/Styles/Typography/Foundation/TextStyle-Header-HomeBasePower-XS.TextStyle-Header-HomeBasePower-XS_C
   )
   > SetStyle(self=@n10, InStyle=/FortUILibrary/Styles/Typography/Foundation/TextStyle-Base-XS-B.TextStyle-Base-XS-B_C)
   > n20 = SetBrush(
    self=Get Image_Power()
    InBrush=Make Slate Brush(
     TintColor=(SpecifiedColor=(R=1,G=1,B=1,A=1),ColorUseRule=UseColor_Specified)
     DrawAs=Image
     Tiling=NoTile
     Mirroring=NoMirror
     ImageSize=(X=24,Y=24)
     Margin=(Left=0,Top=0,Right=0,Bottom=0)
     ResourceObject=/Game/UI/Foundation/Textures/Icons/Shell/T-Icon-Power-32
     OutlineSettings=(CornerRadii=(X=0,Y=0,Z=0,W=0),Color=(SpecifiedColor=(R=0,G=0,B=0,A=0),ColorUseRule=UseColor_Specified),Width=0,RoundingType=HalfHeightRadius,bUseBrushTransparency=False)
    )
   )
   SetVisibility(execute=@n1.else, self=[@n2, @n3, @n4], InVisibility=Collapsed)
   > SetStyle(self=[@n5, @n6], InStyle=/FortUILibrary/Styles/Typography/Foundation/TextStyle-Header-XL.TextStyle-Header-XL_C)
   > n21 = Branch(Condition=@n7)
   n22 = Get TeamRatingValue()
   > SetStyle(self=@n22, InStyle=&TextStyle-Header-L_C)
   n23 = Get BaseRatingValue()
   > SetStyle(self=@n23, InStyle=&TextStyle-Header-L_C)
   > n24 = SetBrush(
    self=Get Image_Power()
    InBrush=Make Slate Brush(
     TintColor=(SpecifiedColor=(R=1,G=1,B=1,A=1),ColorUseRule=UseColor_Specified)
     DrawAs=Image
     Tiling=NoTile
     Mirroring=NoMirror
     ImageSize=(X=56,Y=56)
     Margin=(Left=0,Top=0,Right=0,Bottom=0)
     ResourceObject=&T-Icon-Power-64
     OutlineSettings=(CornerRadii=(X=0,Y=0,Z=0,W=0),Color=(SpecifiedColor=(R=0,G=0,B=0,A=0),ColorUseRule=UseColor_Specified),Width=0,RoundingType=HalfHeightRadius,bUseBrushTransparency=False)
    )
   )
   SetStyle(execute=@n21.else, self=@n22, InStyle=&TextStyle-Base-XS-B_C)
   > SetStyle(self=@n23, InStyle=&TextStyle-Base-XS-B_C)
   > n25 = SetBrush(
    self=Get Image_Power()
    InBrush=Make Slate Brush(
     TintColor=(SpecifiedColor=(R=1,G=1,B=1,A=1),ColorUseRule=UseColor_Specified)
     DrawAs=Image
     Tiling=NoTile
     Mirroring=NoMirror
     ImageSize=(X=24,Y=24)
     Margin=(Left=0,Top=0,Right=0,Bottom=0)
     ResourceObject=&T-Icon-Power-32
     OutlineSettings=(CornerRadii=(X=0,Y=0,Z=0,W=0),Color=(SpecifiedColor=(R=0,G=0,B=0,A=0),ColorUseRule=UseColor_Specified),Width=0,RoundingType=HalfHeightRadius,bUseBrushTransparency=False)
    )
   )
  ]}
]
variables [
 FontColor {type=SlateColor, default {colorUseRule=UseColor_Specified, specifiedColor=<rgba 1 1 1 1>}}
 InterpDuration {type=number, default=1}
 Power Rating Num {type=integer}
 PowerBackingColor {type=LinearColor, default=<rgba 0.131 0.1714 0.3278 1>}
 PowerIncreasedSound {type=SoundBase, default=/Game/Sounds/Fort_UI_Sounds/FrontEnd/New_FrontEnd/UI_ShowHeaderText_Appear_Cue}
 ShowDualPower {type=boolean, default=true}
 ShowGlow {type=boolean, default=true}
 ShowProgress {type=boolean, default=true}
 UseLargeStyle {type=boolean, default=true}
 UseOversizedMode {type=boolean}
]
widgets {
 info {
  inheritedWidgetCount=0
  parentClass=/Script/FortniteUI.FortPlayerPowerRating
  rootWidgetClass=&CommonWidgetSwitcherLegacy
  widgetCount=34
 }
 widgets [
  ModeTypeSwitcher {
   widgetClassPath=&CommonWidgetSwitcherLegacy
   bIsVariable=true
   widget=ModeTypeSwitcher
   properties {transitionDuration=0}
  }
  RegularMode {widgetClassPath=&HorizontalBox, bIsVariable=true, parent=ModeTypeSwitcher, widget=RegularMode, slot {class=WidgetSwitcherSlot}}
  Overlay_1 {widgetClassPath=&Overlay, parent=RegularMode, widget=Overlay_1, properties {visibility=Visible}, slot {class=HorizontalBoxSlot}}
  PowerIconGlow {
   widgetClassPath=&Image
   bIsVariable=true
   parent=Overlay_1
   widget=PowerIconGlow
   properties {
    visibility=Collapsed
    brush {
     bIsDynamicallyLoaded=false
     drawAs=Image
     imageType=NoImage
     mirroring=NoMirror
     resourceName=None
     resourceObject=/FortUILibrary/Materials/Foundation/M-HomeBasePower-OuterGlow
     tiling=NoTile
     imageSize=<4 4>
     margin=<ltrb 0 0 0 0>
     outlineSettings {
      bUseBrushTransparency=false
      roundingType=HalfHeightRadius
      width=0
      color {colorUseRule=UseColor_Specified, specifiedColor=<rgba 0 0 0 0>}
      cornerRadii=<0 0 0 0>
     }
     tintColor {colorUseRule=UseColor_Specified, specifiedColor=<rgba 1 1 1 1>}
     uVRegion {bIsValid=false, max=<0 0>, min=<0 0>}
    }
   }
   slot {class=OverlaySlot, properties {horizontalAlignment=HAlign_Fill, verticalAlignment=VAlign_Fill}}
  }
  VerticalBox_1 {
   widgetClassPath=&VerticalBox
   parent=RegularMode
   widget=VerticalBox_1
   slot {class=HorizontalBoxSlot, properties {horizontalAlignment=HAlign_Left}}
  }
  HorizontalBox_0 {
   widgetClassPath=&HorizontalBox
   parent=VerticalBox_1
   widget=HorizontalBox_0
   slot {class=VerticalBoxSlot, properties {horizontalAlignment=HAlign_Left, verticalAlignment=VAlign_Center}}
  }
  Overlay_0 {widgetClassPath=&Overlay, parent=HorizontalBox_0, widget=Overlay_0, slot {class=HorizontalBoxSlot}}
  BuffedRatingOuterGlow {
   widgetClassPath=&Image
   bIsVariable=true
   parent=Overlay_0
   widget=BuffedRatingOuterGlow
   properties {
    visibility=Collapsed
    brush {
     bIsDynamicallyLoaded=false
     drawAs=Image
     imageType=NoImage
     mirroring=NoMirror
     resourceName=None
     resourceObject=&M-HomeBasePower-OuterGlow
     tiling=NoTile
     imageSize=<4 4>
     margin=<ltrb 0 0 0 0>
     outlineSettings {
      bUseBrushTransparency=false
      roundingType=HalfHeightRadius
      width=0
      color {colorUseRule=UseColor_Specified, specifiedColor=<rgba 0 0 0 0>}
      cornerRadii=<0 0 0 0>
     }
     tintColor {colorUseRule=UseColor_Specified, specifiedColor=<rgba 1 1 1 1>}
     uVRegion {bIsValid=false, max=<0 0>, min=<0 0>}
    }
    renderTransform {angle=0, scale=<1.8 1>, shear=<0 0>, translation=<0 0>}
   }
   slot {class=OverlaySlot, properties {horizontalAlignment=HAlign_Fill, verticalAlignment=VAlign_Fill}}
  }
  Image_Power {
   widgetClassPath=&Image
   bIsVariable=true
   parent=Overlay_0
   widget=Image_Power
   properties {
    renderOpacity=0
    visibility=Hidden
    brush {
     bIsDynamicallyLoaded=false
     drawAs=Image
     imageType=NoImage
     mirroring=NoMirror
     resourceName=None
     resourceObject=&T-Icon-Power-64
     tiling=NoTile
     imageSize=<64 56>
     margin=<ltrb 0 0 0 0>
     outlineSettings {
      bUseBrushTransparency=false
      roundingType=HalfHeightRadius
      width=0
      color {colorUseRule=UseColor_Specified, specifiedColor=<rgba 0 0 0 0>}
      cornerRadii=<0 0 0 0>
     }
     tintColor {colorUseRule=UseColor_Specified, specifiedColor=<rgba 1 1 1 1>}
     uVRegion {bIsValid=false, max=<0 0>, min=<0 0>}
    }
   }
   slot {class=OverlaySlot, properties {horizontalAlignment=HAlign_Center, verticalAlignment=VAlign_Center}}
  }
  Overlay_6 {
   widgetClassPath=&Overlay
   parent=Overlay_0
   widget=Overlay_6
   slot {class=OverlaySlot, properties {verticalAlignment=VAlign_Center, padding=<ltrb 0.5 0.5 0.5 0.5>}}
  }
  Moonbeam_Power_Back {
   widgetClassPath=&Image
   bIsVariable=true
   parent=Overlay_6
   widget=Moonbeam_Power_Back
   properties {
    brush {
     bIsDynamicallyLoaded=false
     drawAs=Image
     imageType=NoImage
     mirroring=NoMirror
     resourceName=None
     resourceObject=/Game/UI/Foundation/Textures/Icons/Items/T-Items-Power-BG
     tiling=NoTile
     imageSize=<40 24>
     margin=<ltrb 0 0 0 0>
     outlineSettings {
      bUseBrushTransparency=false
      roundingType=HalfHeightRadius
      width=0
      color {colorUseRule=UseColor_Specified, specifiedColor=<rgba 0 0 0 0>}
      cornerRadii=<0 0 0 0>
     }
     tintColor {colorUseRule=UseColor_Specified, specifiedColor=<rgba 1 1 1 1>}
     uVRegion {bIsValid=false, max=<0 0>, min=<0 0>}
    }
    colorAndOpacity=<rgba 0.131 0.1714 0.3278 1>
   }
   slot {class=OverlaySlot, properties {horizontalAlignment=HAlign_Fill, verticalAlignment=VAlign_Fill, padding=<ltrb -4 4 -12 4>}}
  }
  HorizontalBox_4 {
   widgetClassPath=&HorizontalBox
   parent=Overlay_6
   widget=HorizontalBox_4
   slot {class=OverlaySlot, properties {horizontalAlignment=HAlign_Center, verticalAlignment=VAlign_Center}}
  }
  TeamRatingValue {
   widgetClassPath=/Script/CommonUI.CommonNumericTextBlock
   bIsVariable=true
   parent=HorizontalBox_4
   widget=TeamRatingValue
   properties {
    currentNumericValue=1
    interpolationUpdateInterval=0.1
    style=&TextStyle-Header-L_C
    text=125
    font {
     bForceMonospaced=false
     bMaterialIsStencil=false
     fontMaterial=null
     fontObject=&Burbank
     letterSpacing=0
     monospacedWidth=1
     size=48
     skewAmount=0
     typefaceFontName=Burbank-Black
     outlineSettings {
      bApplyOutlineToDropShadows=false
      bMiteredCorners=false
      bSeparateFillAlpha=false
      outlineMaterial=null
      outlineSize=0
      outlineColor=<rgba 0 0 0 1>
     }
    }
    shadowOffset=<0 0>
   }
   slot {class=HorizontalBoxSlot, properties {padding=<ltrb 0 0 4 0>}}
  }
  BaseRatingValue {
   widgetClassPath=&CommonTextBlock
   bIsVariable=true
   parent=HorizontalBox_4
   widget=BaseRatingValue
   properties {
    style=&TextStyle-Header-L_C
    text=(1)
    visibility=Collapsed
    font {
     bForceMonospaced=false
     bMaterialIsStencil=false
     fontMaterial=null
     fontObject=&Burbank
     letterSpacing=0
     monospacedWidth=1
     size=48
     skewAmount=0
     typefaceFontName=Burbank-Black
     outlineSettings {
      bApplyOutlineToDropShadows=false
      bMiteredCorners=false
      bSeparateFillAlpha=false
      outlineMaterial=null
      outlineSize=0
      outlineColor=<rgba 0 0 0 1>
     }
    }
    shadowOffset=<0 0>
   }
   slot {class=HorizontalBoxSlot, properties {horizontalAlignment=HAlign_Center, verticalAlignment=VAlign_Center, padding=<ltrb 2 0 8 0>}}
  }
  RatingBar {
   widgetClassPath=&SizeBox
   bIsVariable=true
   parent=VerticalBox_1
   widget=RatingBar
   properties {bOverride_HeightOverride=true, bOverride_WidthOverride=true, heightOverride=12, minDesiredHeight=12, widthOverride=96}
   slot {class=VerticalBoxSlot, properties {padding=<ltrb -4 -16 0 4>}}
  }
  Overlay_3 {widgetClassPath=&Overlay, parent=RatingBar, widget=Overlay_3, slot {class=SizeBoxSlot}}
  BarBG {
   widgetClassPath=&Image
   parent=Overlay_3
   widget=BarBG
   properties {colorAndOpacity=<rgba 1 1 1 0.4>}
   slot {class=OverlaySlot, properties {horizontalAlignment=HAlign_Fill, verticalAlignment=VAlign_Fill}}
  }
  HomebaseRatingBar {
   widgetClassPath=&/HomebaseRatingBar
   bIsVariable=true
   parent=Overlay_3
   widget=HomebaseRatingBar
   properties {
    flowDirectionPreference=Inherit
    barSectionInfo [
     {
      percent=0
      sectionStyle {
       bGradientBar=true
       barColorOneParamName=Bar1RGBA0
       barColorTwoParamName=Bar1RGBA1
       barParamName=Bar1
       barColorOne=<rgba 1 1 1 1>
       barColorTwo=<rgba 1 1 1 1>
      }
     }
     {
      percent=0
      sectionStyle {
       bGradientBar=true
       barColorOneParamName=Bar2RGBA0
       barColorTwoParamName=Bar2RGBA1
       barParamName=Bar2
       barColorOne=<rgba 0.0953 1 1 1>
       barColorTwo=<rgba 0.0953 1 1 1>
      }
     }
     {
      percent=0
      sectionStyle {
       bGradientBar=true
       barColorOneParamName=Bar3RGBA0
       barColorTwoParamName=Bar3RGBA1
       barParamName=Bar3
       barColorOne=<rgba 0.0529 0.7084 0.1046 1>
       barColorTwo=<rgba 0.0529 0.6939 0.0976 1>
      }
     }
     {
      percent=0
      sectionStyle {
       bGradientBar=true
       barColorOneParamName=NegativeBar4RGBA0
       barColorTwoParamName=NegativeBar4RGBA1
       barParamName=NegativeBar4
       barColorOne=<rgba 0.8796 0.1022 0.0176 1>
       barColorTwo=<rgba 0.935 0.0918 0 1>
      }
     }
    ]
   }
   slot {class=OverlaySlot, properties {horizontalAlignment=HAlign_Fill, verticalAlignment=VAlign_Fill, padding=<ltrb 2 2 2 2>}}
  }
  OversizedMode {widgetClassPath=&VerticalBox, bIsVariable=true, parent=ModeTypeSwitcher, widget=OversizedMode, slot {class=WidgetSwitcherSlot}}
  Overlay_5 {
   widgetClassPath=&Overlay
   parent=OversizedMode
   widget=Overlay_5
   slot {class=VerticalBoxSlot, properties {verticalAlignment=VAlign_Center}}
  }
  BackgroundBlur_0 {
   widgetClassPath=&BackgroundBlur
   parent=Overlay_5
   widget=BackgroundBlur_0
   properties {blurStrength=25}
   slot {class=OverlaySlot}
  }
  Overlay_4 {
   widgetClassPath=&Overlay
   parent=BackgroundBlur_0
   widget=Overlay_4
   slot {class=BackgroundBlurSlot, properties {padding=<ltrb 0 0 0 0>}}
  }
  Backplate-Light {
   widgetClassPath=&Image
   parent=Overlay_4
   widget=Backplate-Light
   properties {
    visibility=HitTestInvisible
    brush {
     bIsDynamicallyLoaded=false
     drawAs=Image
     imageType=NoImage
     mirroring=NoMirror
     resourceName=None
     resourceObject=null
     tiling=NoTile
     imageSize=<128 84>
     margin=<ltrb 0.2844 0 0.5 0>
     outlineSettings {
      bUseBrushTransparency=false
      roundingType=HalfHeightRadius
      width=0
      color {colorUseRule=UseColor_Specified, specifiedColor=<rgba 0 0 0 0>}
      cornerRadii=<0 0 0 0>
     }
     tintColor {colorUseRule=UseColor_Specified, specifiedColor=<rgba 1 1 1 1>}
     uVRegion {bIsValid=false, max=<0 0>, min=<0 0>}
    }
    colorAndOpacity=<rgba 0.4452 0.6939 1 0.25>
   }
   slot {class=OverlaySlot, properties {horizontalAlignment=HAlign_Fill, verticalAlignment=VAlign_Fill}}
  }
  HorizontalBox_1 {
   widgetClassPath=&HorizontalBox
   parent=Overlay_4
   widget=HorizontalBox_1
   slot {class=OverlaySlot, properties {horizontalAlignment=HAlign_Fill, verticalAlignment=VAlign_Fill}}
  }
  Border-LeftBackplate {
   widgetClassPath=&Border
   parent=HorizontalBox_1
   widget=Border-LeftBackplate
   properties {
    horizontalAlignment=HAlign_Right
    background {
     bIsDynamicallyLoaded=false
     drawAs=Box
     imageType=NoImage
     mirroring=NoMirror
     resourceName=None
     resourceObject=null
     tiling=NoTile
     imageSize=<128 128>
     margin=<ltrb 0 0 0.75 0>
     outlineSettings {
      bUseBrushTransparency=false
      roundingType=HalfHeightRadius
      width=0
      color {colorUseRule=UseColor_Specified, specifiedColor=<rgba 0 0 0 0>}
      cornerRadii=<0 0 0 0>
     }
     tintColor {colorUseRule=UseColor_Specified, specifiedColor=<rgba 1 1 1 0>}
     uVRegion {bIsValid=false, max=<0 0>, min=<0 0>}
    }
    padding=<ltrb 24 0 0 0>
   }
   slot {class=HorizontalBoxSlot}
  }
  HorizontalBox_2 {
   widgetClassPath=&HorizontalBox
   parent=Border-LeftBackplate
   widget=HorizontalBox_2
   slot {class=BorderSlot, properties {horizontalAlignment=HAlign_Right, padding=<ltrb 24 0 0 0>}}
  }
  CommonTextBlock_0 {
   widgetClassPath=&CommonTextBlock
   parent=HorizontalBox_2
   widget=CommonTextBlock_0
   properties {
    style=/FortUILibrary/Styles/Typography/Foundation/TextStyle-Header-M-Horizon.TextStyle-Header-M-Horizon_C
    text=POWER
    font {
     bForceMonospaced=false
     bMaterialIsStencil=false
     fontMaterial=/FortUILibrary/Materials/Foundation/M_Font_Horizon_Dynamic
     fontObject=&Burbank
     letterSpacing=0
     monospacedWidth=1
     size=32
     skewAmount=0
     typefaceFontName=Burbank-Black
     outlineSettings {
      bApplyOutlineToDropShadows=true
      bMiteredCorners=false
      bSeparateFillAlpha=false
      outlineMaterial=null
      outlineSize=3
      outlineColor=<rgba 0 0 0 0.1>
     }
    }
    shadowOffset=<0 0>
   }
   slot {class=HorizontalBoxSlot, properties {horizontalAlignment=HAlign_Right, verticalAlignment=VAlign_Center}}
  }
  PowerIcons-Large {
   widgetClassPath=&Image
   parent=HorizontalBox_2
   widget=PowerIcons-Large
   properties {
    visibility=Collapsed
    brush {
     bIsDynamicallyLoaded=false
     drawAs=Image
     imageType=NoImage
     mirroring=NoMirror
     resourceName=None
     resourceObject=/Game/UI/Foundation/Textures/Icons/Shell/T-Icon-Power-128
     tiling=NoTile
     imageSize=<64 64>
     margin=<ltrb 0 0 0 0>
     outlineSettings {
      bUseBrushTransparency=false
      roundingType=HalfHeightRadius
      width=0
      color {colorUseRule=UseColor_Specified, specifiedColor=<rgba 0 0 0 0>}
      cornerRadii=<0 0 0 0>
     }
     tintColor {colorUseRule=UseColor_Specified, specifiedColor=<rgba 1 1 1 1>}
     uVRegion {bIsValid=false, max=<0 0>, min=<0 0>}
    }
    renderTransform {angle=0, scale=<1 1>, shear=<-4 0>, translation=<0 0>}
   }
   slot {class=HorizontalBoxSlot, properties {horizontalAlignment=HAlign_Right, verticalAlignment=VAlign_Center}}
  }
  HorizontalBox_3 {
   widgetClassPath=&HorizontalBox
   parent=HorizontalBox_1
   widget=HorizontalBox_3
   slot {class=HorizontalBoxSlot, properties {horizontalAlignment=HAlign_Left, padding=<ltrb 10 -12 24 -12>, size {sizeRule=Fill, value=1}}}
  }
  Overlay_2 {
   widgetClassPath=&Overlay
   parent=HorizontalBox_3
   widget=Overlay_2
   properties {renderTransform {angle=0, scale=<1 1>, shear=<0 0>, translation=<-8 0>}}
   slot {class=HorizontalBoxSlot, properties {horizontalAlignment=HAlign_Left}}
  }
  Rating-Glow-Large {
   widgetClassPath=&Image
   bIsVariable=true
   parent=Overlay_2
   widget=Rating-Glow-Large
   properties {
    visibility=Collapsed
    brush {
     bIsDynamicallyLoaded=false
     drawAs=Image
     imageType=NoImage
     mirroring=NoMirror
     resourceName=None
     resourceObject=&M-HomeBasePower-OuterGlow
     tiling=NoTile
     imageSize=<32 64>
     margin=<ltrb 0 0 0 0>
     outlineSettings {
      bUseBrushTransparency=false
      roundingType=HalfHeightRadius
      width=0
      color {colorUseRule=UseColor_Specified, specifiedColor=<rgba 0 0 0 0>}
      cornerRadii=<0 0 0 0>
     }
     tintColor {colorUseRule=UseColor_Specified, specifiedColor=<rgba 1 1 1 1>}
     uVRegion {bIsValid=false, max=<0 0>, min=<0 0>}
    }
    renderTransform {angle=0, scale=<2 1>, shear=<0 0>, translation=<0 0>}
   }
   slot {class=OverlaySlot, properties {horizontalAlignment=HAlign_Fill, verticalAlignment=VAlign_Fill}}
  }
  TeamRatingValue-Large {
   widgetClassPath=&CommonNumericTextBlock
   bIsVariable=true
   parent=Overlay_2
   widget=TeamRatingValue-Large
   properties {
    style=&TextStyle-Header-XL_C
    text=0
    font {
     bForceMonospaced=false
     bMaterialIsStencil=false
     fontMaterial=null
     fontObject=&Burbank
     letterSpacing=0
     monospacedWidth=1
     size=72
     skewAmount=0
     typefaceFontName=Burbank-Black
     outlineSettings {
      bApplyOutlineToDropShadows=false
      bMiteredCorners=false
      bSeparateFillAlpha=false
      outlineMaterial=null
      outlineSize=0
      outlineColor=<rgba 0 0 0 1>
     }
    }
    shadowOffset=<0 0>
   }
   slot {class=OverlaySlot}
  }
  BaseRatingValue-Large {
   widgetClassPath=&CommonTextBlock
   bIsVariable=true
   parent=HorizontalBox_3
   widget=BaseRatingValue-Large
   properties {
    style=&TextStyle-Header-XL_C
    text=(0)
    visibility=Collapsed
    font {
     bForceMonospaced=false
     bMaterialIsStencil=false
     fontMaterial=null
     fontObject=&Burbank
     letterSpacing=0
     monospacedWidth=1
     size=72
     skewAmount=0
     typefaceFontName=Burbank-Black
     outlineSettings {
      bApplyOutlineToDropShadows=false
      bMiteredCorners=false
      bSeparateFillAlpha=false
      outlineMaterial=null
      outlineSize=0
      outlineColor=<rgba 0 0 0 1>
     }
    }
    shadowOffset=<0 0>
   }
   slot {class=HorizontalBoxSlot}
  }
  Image_0 {
   widgetClassPath=&Image
   parent=Overlay_5
   widget=Image_0
   properties {brush {
     bIsDynamicallyLoaded=false
     drawAs=Image
     imageType=NoImage
     mirroring=NoMirror
     resourceName=None
     resourceObject=/Game/UI/Frontend/Command/M_UI_SkewBorderFrame
     tiling=NoTile
     imageSize=<32 32>
     margin=<ltrb 0 0 0 0>
     outlineSettings {
      bUseBrushTransparency=false
      roundingType=HalfHeightRadius
      width=0
      color {colorUseRule=UseColor_Specified, specifiedColor=<rgba 0 0 0 0>}
      cornerRadii=<0 0 0 0>
     }
     tintColor {colorUseRule=UseColor_Specified, specifiedColor=<rgba 1 1 1 1>}
     uVRegion {bIsValid=false, max=<0 0>, min=<0 0>}
    }}
   slot {class=OverlaySlot, properties {horizontalAlignment=HAlign_Fill, verticalAlignment=VAlign_Fill}}
  }
 ]
}
